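# Physics asset for the Orange Camo medium male player body cosmetic, covering five dangling flap bones (back, left, right and two front) around the hips.
# Each flap is a box body constrained to the pelvis, with tight twist and swing1 limits and a wide swing2, so it moves like a hinge.
# Only the pelvis and thighs have other bodies, built from spheres and capsules as collision volumes rather than a full-body ragdoll.
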PhysicsAsset /BRCosmetics/Characters/Player/Male/Medium/Bodies/M_MED_Orange_Camo/Meshes/M_MED_Orange_Camo_Body_Physics
bodies {
 dyn_flap_back [{shapeType=Box, extentX=20, extentY=24, extentZ=6.5, center=<12 0 4>}]
 dyn_flap_front_l [{shapeType=Box, extentX=25.5, extentY=16, extentZ=8, center=<13.1677 0.8725 1.246>, rotation=<rot 0 0 30.0002>}]
 dyn_flap_front_r [{shapeType=Box, extentX=20, extentY=10, extentZ=8, center=<9.8257 -0.4034 0.6989>, rotation=<rot 0 0 -29.9992>}]
 dyn_flap_l [{shapeType=Box, extentX=20, extentY=15, extentZ=5, center=<10 0 5>}]
 dyn_flap_r [{shapeType=Box, extentX=20, extentY=15, extentZ=5, center=<10 0 5>}]
 pelvis [
  {shapeType=Sphere, radius=4.25, center=<-5.903 6.5771 -0.0003>}
  {shapeType=Capsule, length=0.5, radius=0.5}
  {shapeType=Capsule, length=15, radius=9.8, shapeName=pelvis_capsule, center=<-4.2662 -4.7342 0>}
 ]
 thigh_l [
  {shapeType=Sphere, radius=5, center=<-13.5206 7.4666 0.0003>}
  {shapeType=Capsule, length=33.5352, radius=10.0844, center=<-17.5057 -1.1796 0>, rotation=<rot 90 177.7266 0>}
 ]
 thigh_r [
  {shapeType=Sphere, radius=5, center=<13.5206 -7.4666 -0.0003>}
  {shapeType=Capsule, length=33.5352, radius=10.0844, center=<17.5057 1.1796 0>, rotation=<rot 90 -2.2734 0>}
 ]
}
constraints [
 {bone1Name=dyn_flap_front_l, bone2Name=pelvis, swing1LimitDegrees=5, swing2LimitDegrees=90, twistLimitDegrees=5}
 {bone1Name=dyn_flap_l, bone2Name=pelvis, swing1LimitDegrees=5, swing2LimitDegrees=90, twistLimitDegrees=5}
 {bone1Name=dyn_flap_back, bone2Name=pelvis, swing1LimitDegrees=5, swing2LimitDegrees=60, twistLimitDegrees=5}
 {bone1Name=dyn_flap_front_r, bone2Name=pelvis, swing1LimitDegrees=5, swing2LimitDegrees=90, twistLimitDegrees=5}
 {bone1Name=dyn_flap_r, bone2Name=pelvis, swing1LimitDegrees=5, swing2LimitDegrees=90, twistLimitDegrees=5}
]
