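# Animation Blueprint for the male medium Orange Camo body cosmetic, adding secondary motion to a pose copied from the attached parent mesh.
# A clavicle RBF Pose Driver corrects the collar bone, then a pelvis-space RigidBody node simulates dynamics into the output.
# Skydiving, parachute and crouch events reset dynamics, some after delays.
# Upper-arm driver, targeting/crouch blends and front-flap bone controls sit outside the output chain.
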AnimBlueprint /BRCosmetics/Characters/Player/Male/Medium/Bodies/M_MED_Orange_Camo/Meshes/M_MED_Orange_Camo_Body_AnimBP
target_skeleton=&/M_MED_Orange_Camo_Body_Skeleton
graphs [
 AnimGraph {nodes [
   Blend Poses by bool(bActiveValue=Get bIsCrouching()) {blendType=Linear}
   n1 = Pose Driver - Source: upperarm_l!(SourcePose:Pose Link Structure=Use cached pose 'cpfm Cache'()) -> (Pose:Pose Link Structure) {
    is=title AnimGraphNode_PoseDriver
    poseAsset=&/M_MED_Orange_Camo_RBF_PoseAsset
    onlyDriveBones [{boneName=dyn_collar}]
    poseTargets [
     {
      bApplyCustomCurve=false
      bIsHidden=false
      distanceMethod=DefaultMethod
      drivenName=Pose_0
      functionType=DefaultFunction
      targetScale=1
      boneTransforms [{targetRotation=<rot 0 0 0>, targetTranslation=<0 0 0>}, {targetRotation=<rot 0 0 0>, targetTranslation=<0 0 0>}]
      customCurve {defaultValue=3.4028234663852886e+38, postInfinityExtrap=RCCE_Constant, preInfinityExtrap=RCCE_Constant, keys []}
      targetRotation=<rot 0 0 0>
     }
     {
      bApplyCustomCurve=false
      bIsHidden=false
      distanceMethod=DefaultMethod
      drivenName=Pose_1
      functionType=DefaultFunction
      targetScale=0.0215
      boneTransforms [
       {targetRotation=<rot -51.299 35.277 5.8965>, targetTranslation=<15.5111 0 0>}
       {targetRotation=<rot -74.1463 26.5666 -27.0012>, targetTranslation=<-15.511 0 0>}
      ]
      customCurve {defaultValue=3.4028234663852886e+38, postInfinityExtrap=RCCE_Constant, preInfinityExtrap=RCCE_Constant, keys []}
      targetRotation=<rot 0 0 0>
     }
     {
      bApplyCustomCurve=false
      bIsHidden=false
      distanceMethod=DefaultMethod
      drivenName=Pose_6
      functionType=DefaultFunction
      targetScale=0.0215
      boneTransforms [
       {targetRotation=<rot -51.3423 37.6428 5.5045>, targetTranslation=<15.5111 0 0>}
       {targetRotation=<rot -73.6978 26.9238 -27.286>, targetTranslation=<-15.511 0 0>}
      ]
      customCurve {defaultValue=3.4028234663852886e+38, postInfinityExtrap=RCCE_Constant, preInfinityExtrap=RCCE_Constant, keys []}
      targetRotation=<rot 0 0 0>
     }
     {
      bApplyCustomCurve=false
      bIsHidden=false
      distanceMethod=DefaultMethod
      drivenName=Pose_7
      functionType=DefaultFunction
      targetScale=0.8204
      boneTransforms [
       {targetRotation=<rot -8.9273 26.0392 -80.0151>, targetTranslation=<15.5111 0 0>}
       {targetRotation=<rot -36.8576 112.2266 -68.8487>, targetTranslation=<-15.511 0 0>}
      ]
      customCurve {defaultValue=3.4028234663852886e+38, postInfinityExtrap=RCCE_Constant, preInfinityExtrap=RCCE_Constant, keys []}
      targetRotation=<rot 0 0 0>
     }
     {
      bApplyCustomCurve=false
      bIsHidden=false
      distanceMethod=DefaultMethod
      drivenName=Pose_9
      functionType=DefaultFunction
      targetScale=0.1056
      boneTransforms [
       {targetRotation=<rot 5.6173 57.9806 -99.4092>, targetTranslation=<15.5127 -0.0002 0.0018>}
       {targetRotation=<rot 32.4312 100.4243 -80.0644>, targetTranslation=<-15.5109 0.0003 0.0008>}
      ]
      customCurve {defaultValue=3.4028234663852886e+38, postInfinityExtrap=RCCE_Constant, preInfinityExtrap=RCCE_Constant, keys []}
      targetRotation=<rot 0 0 0>
     }
    ]
    rBFParams {distanceMethod=Quaternion, radius=60}
    sourceBones [{boneName=upperarm_l}, {boneName=upperarm_r}]
   }
   Blend Poses by bool(BlendPose_0=@n1, bActiveValue=Get bIsTargeting()) {blendType=Linear}
   "Constraint - dyn_flap_front_l (P)"(Alpha=1, ConstraintWeights_0=1) {
    alphaBoolBlend {blendOption=Linear}
    boneToModify {boneName=dyn_flap_front_l}
    constraintSetup [{targetBone {boneName=thigh_l}}]
    constraintWeights [1]
   }
   n2 = Copy Bone Delta - Source Bone: thigh_l - Target Bone: dyn_flap_front_l!(ComponentPose:Component Space Pose Link Structure, Alpha:float=1) -> (Pose:Component Space Pose Link Structure) {
    is=title AnimGraphNode_CopyBoneDelta
    bCopyRotation=true
    rotationMultiplier=2
    alphaBoolBlend {blendOption=Linear}
    sourceBone {boneName=thigh_l}
    targetBone {boneName=dyn_flap_front_l}
   }
   Copy Bone Delta - Source Bone: thigh_r - Target Bone: dyn_flap_front_r!(ComponentPose:Component Space Pose Link Structure=@n2, Alpha:float=1) -> (Pose:Component Space Pose Link Structure) {
    is=title AnimGraphNode_CopyBoneDelta
    bCopyRotation=true
    rotationMultiplier=2
    alphaBoolBlend {blendOption=Linear}
    sourceBone {boneName=thigh_r}
    targetBone {boneName=dyn_flap_front_r}
   }
   "Transform (Modify) Bone - Bone: dyn_flap_front_l"(Alpha="To Float (Boolean)"(InBool=Get bIsCrouching()), Translation="0,0,20", Rotation="0,0,0", Scale="1,1,1") {
    rotationMode=BMM_Additive
    rotationSpace=BCS_BoneSpace
    translationMode=BMM_Additive
    translationSpace=BCS_BoneSpace
    alphaBoolBlend {blendOption=Linear}
    boneToModify {boneName=dyn_flap_front_l}
    translation {z=20}
   }
   n3 = Pose Driver - Source: clavicle_l!(SourcePose:Pose Link Structure=Use cached pose 'cpfm Cache'()) -> (Pose:Pose Link Structure) {
    is=title AnimGraphNode_PoseDriver
    poseAsset=&M_MED_Orange_Camo_RBF_PoseAsset
    onlyDriveBones [{boneName=dyn_collar}]
    poseTargets [
     {
      bApplyCustomCurve=false
      bIsHidden=false
      distanceMethod=DefaultMethod
      drivenName=Pose_0
      functionType=DefaultFunction
      targetScale=1
      boneTransforms [{targetRotation=<rot 0 0 0>, targetTranslation=<0 0 0>}, {targetRotation=<rot 0 0 0>, targetTranslation=<0 0 0>}]
      customCurve {defaultValue=3.4028234663852886e+38, postInfinityExtrap=RCCE_Constant, preInfinityExtrap=RCCE_Constant, keys []}
      targetRotation=<rot 0 0 0>
     }
     {
      bApplyCustomCurve=false
      bIsHidden=false
      distanceMethod=DefaultMethod
      drivenName=Pose_7
      functionType=DefaultFunction
      targetScale=1
      boneTransforms [
       {targetRotation=<rot -51.4969 5.6838 2.5287>, targetTranslation=<4.3581 -0.1465 -2.3223>}
       {targetRotation=<rot -68.9748 -127.5875 -53.9473>, targetTranslation=<4.3577 -0.1466 2.3224>}
      ]
      customCurve {defaultValue=3.4028234663852886e+38, postInfinityExtrap=RCCE_Constant, preInfinityExtrap=RCCE_Constant, keys []}
      targetRotation=<rot 0 0 0>
     }
    ]
    rBFParams {distanceMethod=Quaternion, function=Linear, radius=40}
    sourceBones [{boneName=clavicle_l}, {boneName=clavicle_r}]
   }
   n4 = RigidBody(ComponentPose=Local To Component(LocalPose=@n3)) {
    bOverrideWorldGravity=true
    overridePhysicsAsset=&/M_MED_Orange_Camo_Body_Physics
    simulationSpace=BaseBoneSpace
    alphaBoolBlend {blendOption=Linear}
    baseBoneRef {boneName=pelvis}
    componentAppliedLinearAccClamp=<2000 2000 2000>
    overrideWorldGravity {z=-980}
    simSpaceSettings {dampingAlpha=0.5, velocityScaleZ=0.8, worldAlpha=1}
   }
   Output Pose(Result=Component To Local(ComponentPose=@n4)) {layerGroup=None}
   Save cached pose 'cpfm Cache'(Pose=Copy Pose From Mesh() {bUseAttachedParent=true}) {cacheName=cpfm Cache}
   Sequence Player 'M_MED_Orange_Camo_RBF'() {playRate=0.01, sequence=&/M_MED_Orange_Camo_RBF}
  ]}
 EventGraph {nodes [
   Event OnParachuteOpened()
   > n1 = Delay(Duration=1.5)
   n2 = ResetDynamics(execute=[@n1.then, @n6.then, @n3.then, @n4.then], InTeleportType=TeleportPhysics)
   ResetDynamics(execute=[@n2.then, @n8.OutputPin], InTeleportType=ResetPhysics)
   n3 = Event OnBeginSkydiving()
   n4 = Event OnEndSkydiving()
   n5 = Event OnEndCrouching()
   n6 = Delay(execute=[@n5.then, @n7.then], Duration=0.125)
   n7 = Event OnBeginCrouching()
   n8 = Reroute Node()
   TryGetPawnOwner()
   Event BlueprintUpdateAnimation()
  ]}
]
variables []
